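# Masked, lit material for the Arrow vehicle body cosmetic, usable on skeletal meshes and with mesh deformers.
# The chassis Fx diffuse texture drives base color, and the red channel of a linear-color mask texture drives opacity mask.
# These combine with constant metallic, specular and roughness in a material attributes set, which feeds the shared MF_Vehicle_FNPostFX vehicle function.
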Material /VehicleCosmetics/Bodies/Arrow/Materials/MAT_Arrow_Alphamask
graph {
 properties {
  bAutomaticallySetUsageInEditor=false
  bUseMaterialAttributes=true
  bUsedWithMeshDeformer=true
  bUsedWithSkeletalMesh=true
  blendMode=BLEND_Masked
  usedShadingModels=Default Lit
 }
 nodes [
  n1 = Constant() {r=0.5}
  n2 = SetMaterialAttributes!(
   MaterialAttributes
   Base Color=TextureSample() {texture=/VehicleCosmetics/Bodies/Arrow/Textures/Chassis/T_Arrow_Fx_D}.RGB
   Metallic=Constant()
   Specular=@n1
   Roughness=@n1
   Opacity Mask=TextureSample() {samplerType=SAMPLERTYPE_LinearColor, texture=/VehicleCosmetics/Bodies/Arrow/Textures/Chassis/T_Arrow_Fx_Mask}.R
  ) {attributeSetTypes [
    #1
    #2
    #3
    #4
    #5
   ]}
  MaterialFunctionCall!(In=@n2) {materialFunction=/VehicleCosmetics/Materials/MaterialFunctions/MF_Vehicle_FNPostFX}
 ]
}
