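# Material function for the GoldenLychee01 Arrow body skin, setting animated, speed-reactive base color and emissive, plus metal, roughness and normal.
# Base color layers the skin texture, Fresnel shading and fake AO with panned masks and noise-distorted UVs.
# A Speedometer parameter, replaceable by a debug sine, drives masked color blends, while Perlin noise shapes a red-orange emissive overlay.
# The normal is carbon fiber, rotated differently per UV1 half.
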MaterialFunction /VehicleCosmetics/Bodies/Arrow/Skins/MF_Skin_ArrowGoldenLychee01
graph {nodes [
  n1 = SetMaterialAttributes!(
   MaterialAttributes
   BaseColor=NamedRerouteUsage()
   Metallic=ScalarParameter() {defaultValue=0.5, parameterName=Metal}
   Roughness=ScalarParameter() {defaultValue=0.375, parameterName=Roughness}
   EmissiveColor=NamedRerouteUsage()
   Normal=NamedRerouteUsage()
   AmbientOcclusion
  ) {attributeSetTypes [
    #1
    #2
    #3
    #4
    #5
    #6
   ]}
  FunctionOutput(None=@n1)
  n2 = NamedRerouteUsage()
  n3 = Desaturation(None=@n2, Fraction=Constant() {r=0.2})
  Multiply(A=@n3) {constB=4}
  NamedRerouteDeclaration(None=MaterialFunctionCall!(Fresnel Power=Constant() {r=3}, Normal=NamedRerouteUsage()) {materialFunction=/VehicleCosmetics/Materials/MaterialFunctions/MF_Fresnel_Prime}.Result) {name=Fresnel, nodeColor=<rgba 0.7344 0 1 1>}
  NamedRerouteDeclaration() {name=MetalMask, nodeColor=<rgba 0 1 0.4141 1>}
  n4 = VectorParameter() {desc="Alpha = Emissive Color Brightness", group=Colors, parameterName=BaseColor, defaultValue=<rgba 0 0 1 1>}
  n5 = ScalarParameter() {defaultValue=1, parameterName=Override Colors}
  NamedRerouteDeclaration(None=LinearInterpolate(A=@n4.RGB, B=VectorParameter() {parameterName=Override TeamColor, defaultValue=<rgba 1 0.01 0.01 1>}.RGB, Alpha=@n5)) {name=TeamColor, nodeColor=<rgba 0.003 0.0179 1 1>}
  NamedRerouteDeclaration(None=LinearInterpolate(
    A=VectorParameter() {group=Colors, parameterName=AccentColor, defaultValue=<rgba 0 1 0.0262 1>}.RGB
    B=VectorParameter() {parameterName=Override CustomColor, defaultValue=<rgba 0 0.15 1 1>}.RGB
    Alpha=@n5
   )) {name=CustomColor, nodeColor=<rgba 0 1 0.0625 1>}
  NamedRerouteDeclaration(None=LinearInterpolate(
    A=VectorParameter() {group=Colors, parameterName=PaintColor, defaultValue=<rgba 1 0.0137 0.0137 1>}.RGB
    B=VectorParameter() {parameterName=Override PaintColor, defaultValue=<rgba 0.0774 0.213 1 1>}.RGB
    Alpha=@n5
   )) {name=PaintColor, nodeColor=<rgba 1 0 0.0199 1>}
  NamedRerouteDeclaration(None=Time() {bOverride_Period=true, period=500}) {name=time, nodeColor=<rgba 1 0.9285 0 1>}
  NamedRerouteDeclaration(None=TextureSample() {texture=&/T_ArrowGoldenLychee01_RGB}.RGBA) {name=RGBAMask, nodeColor=<rgba 0.125 0 1 1>}
  n6 = Constant() {r=0.75}
  n7 = NamedRerouteUsage()
  n8 = Multiply(A=OneMinus(None=MaterialFunctionCall(Fresnel Power=@n6, Normal=@n7) {materialFunction=&MF_Fresnel_Prime}.Result)) {constB=0.8}
  n9 = ComponentMask(None=@n7) {g=true, r=true}
  n10 = OneMinus(None=Saturate(None=Desaturation(None=@n9))) {desc=Fake AO}
  n11 = MaterialFunctionCall(Fresnel Power=Constant() {r=1}, Normal=@n7) {materialFunction=&MF_Fresnel_Prime}
  n12 = Max(A=ConstantBiasScale(None=OneMinus(None=@n11.Result)) {bias=-0.9, scale=1.25}, B=Subtract(A=@n11.Result) {constB=0.75})
  n13 = LinearInterpolate(
   A=Multiply(A=TextureSample() {texture=&/T_ArrowGoldenLychee01_BC}.RGB, B=Multiply(A=@n8, B=@n10) {constB=0.8}) {constB=0.8}
   Alpha=Saturate(None=Multiply(A=@n12) {constB=0.8})
  ) {constB=0.42}
  n14 = NamedRerouteUsage()
  n15 = ComponentMask(None=TextureCoordinate() {coordinateIndex=1}) {g=true}
  n16 = OneMinus(None=@n15)
  n17 = Saturate(None=ConstantBiasScale(None=@n16) {bias=-0.4, scale=2})
  n18 = TextureCoordinate() {coordinateIndex=1, uTiling=5, vTiling=5}
  n19 = NamedRerouteUsage()
  n20 = Multiply(
   A=@n17
   B=TextureSample(UVs=Panner(Coordinate=@n18, Time=@n19) {bFractionalPart=true, speedX=-0.75, speedY=0.5}) {texture=&T_ArrowGoldenLychee01_RGB}.B
  )
  n21 = NamedRerouteUsage()
  n22 = Saturate(None=Subtract(A=Multiply(A=@n20, B=@n21)) {constB=0.01})
  n23 = TextureSample(UVs=Panner(Coordinate=TextureCoordinate() {coordinateIndex=1, uTiling=3, vTiling=3}, Time=NamedRerouteUsage()) {bFractionalPart=true, speedX=0.02, speedY=0.5}) {texture=&T_ArrowGoldenLychee01_RGB}
  n24 = Multiply(A=@n23.B, B=@n17)
  n25 = OneMinus(None=@n21)
  n26 = Multiply(A=Subtract(A=Subtract(A=@n24, B=@n25)) {constB=0.01}) {constB=10}
  n27 = ConstantBiasScale(None=@n15) {bias=-0.5, scale=2}
  n28 = ConstantBiasScale(None=OneMinus(None=Abs(None=@n27))) {bias=-0.6, scale=3}
  n29 = Subtract(A=Multiply(A=@n23.A, B=Saturate(None=@n28)), B=@n25)
  n30 = LinearInterpolate(
   A=LinearInterpolate(A=LinearInterpolate(A=@n13, B=@n14, Alpha=@n22) {constB=0.42}, B=@n14, Alpha=Saturate(None=@n26)) {constB=0.42}
   Alpha=Saturate(None=Multiply(A=@n29) {constB=10})
  ) {constB=0}
  n31 = LinearInterpolate(Alpha=@n16) {constA=25, constB=50}
  n32 = MaterialFunctionCall!(
   Texture Coordinates
   Tiling X
   Tiling Y
   PanSpeedX
   PanSpeedY
   Chaos Speed
   Angle=LinearInterpolate(A=Multiply(A=@n31) {constB=-1}, B=@n31, Alpha=NamedRerouteUsage()) {constA=25, constB=50}
   Rotation Speed
   Sharpen
  ) {materialFunction=/VehicleCosmetics/Materials/MaterialFunctions/MF_Noisemaker_AnisotropicSmooth}
  n33 = NamedRerouteUsage()
  n34 = Multiply(A=@n32.Aniso2, B=@n33)
  n35 = TextureSample(UVs=Add(A=TextureCoordinate(), B=Multiply(A=@n34) {constB=0.0075})) {texture=&T_ArrowGoldenLychee01_RGB}
  n36 = Saturate(None=Subtract(A=@n35.G, B=Multiply(A=@n32.Aniso3, B=@n33)))
  n37 = Saturate(None=@n3)
  n38 = TextureSample() {texture=&T_ArrowGoldenLychee01_RGB}
  n39 = StaticSwitchParameter(
   True=Sine(None=Time()) {period=5}
   False=ScalarParameter() {desc="Speedometer - uses percentage\r\nof speed from VCID", parameterName=Speedometer}
  ) {parameterName=SpeedDebug}
  n40 = NamedRerouteDeclaration(None=Saturate(None=ConstantBiasScale(None=@n39) {bias=0, desc=Balance copied from Pixie, scale=1})) {name=Speed, nodeColor=<rgba 1 0 0.4922 1>}
  n41 = LinearInterpolate(A=@n2, B=@n37, Alpha=Multiply(A=Multiply(A=@n38.R, B=@n40)) {constB=0.5})
  n42 = TextureSample() {texture=&T_ArrowGoldenLychee01_BC}
  n43 = Multiply(A=@n40, B=@n42.A)
  NamedRerouteDeclaration(None=LinearInterpolate(A=LinearInterpolate(A=@n30, Alpha=@n36) {constB=0.6, desc=diffuse}, B=@n41, Alpha=@n43) {desc=Diffuse}) {name=BaseColor, nodeColor=<rgba 1 0.0938 0 1>}
  n44 = ComponentMask(None=Floor(None=TextureCoordinate() {coordinateIndex=1, uTiling=2, vTiling=2})) {r=true}
  n45 = Rotator(
   Coordinate=TextureCoordinate() {coordinateIndex=1, uTiling=10, vTiling=10}
   Time=LinearInterpolate(Alpha=NamedRerouteDeclaration(None=@n44) {name=SymmetricalSplit, nodeColor=<rgba 0.6406 1 0 1>}) {constA=-6.5, constB=0}
  )
  NamedRerouteDeclaration(None=LinearInterpolate(
    A=Constant3Vector() {constant=<rgba 0 0 1 0>}
    B=TextureSample(UVs=@n45) {samplerType=SAMPLERTYPE_Normal, texture=/VehicleCosmetics/Textures/UniversalTextures/T_CarbonFiber_Flipped_N}.RGB
    Alpha=ScalarParameter() {defaultValue=0.5, parameterName=NormalIntensity}
   )) {name=NormalTexture, nodeColor=<rgba 0.299 0.3454 1 1>}
  n46 = Multiply(A=@n35.G) {constB=3}
  n47 = LinearInterpolate(B=@n2, Alpha=Multiply(A=Saturate(None=@n46), B=@n33) {constB=3}) {constB=0.6}
  n48 = Multiply(A=@n36, B=@n33) {constB=3}
  n49 = Constant3Vector() {constant=<rgba 0.98 0.1255 0 0>}
  n50 = TextureCoordinate()
  n51 = OneMinus(None=ComponentMask(None=TextureCoordinate() {coordinateIndex=1}) {g=true})
  n52 = ConstantBiasScale(None=@n51) {bias=-0.15, scale=5}
  n53 = Constant() {r=2}
  n54 = Constant() {r=3}
  n55 = Constant() {r=1}
  n56 = Constant() {r=4}
  n57 = Multiply(A=Multiply(
    A=Saturate(None=@n52)
    B=MaterialFunctionCall!(Texture Coordinates, Tiling X=@n53, Tiling Y=@n54, PanSpeedX=@n55, PanSpeedY=@n54, Angle, RotateSpeed, Balance=@n56, Chaos=@n55) {materialFunction=/VehicleCosmetics/Materials/MaterialFunctions/MF_Noisemaker_Perlin}.Perlin3
   )) {constB=-0.02}
  n58 = Subtract(A=TextureSample(UVs=Add(A=@n50, B=@n57)) {texture=&T_ArrowGoldenLychee01_RGB}.R, B=@n42.A)
  n59 = Constant() {r=6}
  n60 = Constant() {r=4}
  n61 = Constant() {r=0.5}
  n62 = Constant() {r=2}
  n63 = Saturate(None=Multiply(A=MaterialFunctionCall(Tiling X=@n59, Tiling Y=@n60, PanSpeedX=@n61, PanSpeedY=@n62, Chaos=@n61) {materialFunction=&MF_Noisemaker_Perlin}.Perlin3) {constB=0.5})
  n64 = NamedRerouteUsage()
  n65 = Multiply(A=Saturate(None=Subtract(A=@n58, B=@n63)), B=Multiply(A=@n51, B=ConstantBiasScale(None=@n64) {bias=-0.5, scale=10}))
  n66 = Constant()
  n67 = MaterialFunctionCall(Tiling X=Constant() {r=3}, Tiling Y=Constant() {r=4}, PanSpeedX=@n66, PanSpeedY=Constant() {r=5}, Chaos=@n66) {materialFunction=&MF_Noisemaker_Perlin}
  n68 = Multiply(A=Multiply(A=@n65, B=ConstantBiasScale(None=@n67.Perlin1))) {constB=5}
  n69 = Clamp(None=Multiply(A=@n65) {constB=10}) {maxDefault=0.5}
  n70 = LinearInterpolate(
   A=LinearInterpolate(A=LinearInterpolate(A=@n47, B=@n37, Alpha=@n48) {constB=0.6, desc=Emissive}, B=@n41, Alpha=@n43) {desc=Emissive}
   B=LinearInterpolate(A=@n49, B=Add(A=@n49) {constB=0.3}, Alpha=@n65)
   Alpha=Clamp(None=Max(A=@n68, B=@n69)) {maxDefault=3}
  ) {constB=0.6, desc=Emissive}
  NamedRerouteDeclaration(None=Multiply(A=Saturate(None=@n70), B=ScalarParameter() {defaultValue=1, parameterName=EmissiveIntensity})) {name=Emissive, nodeColor=<rgba 0 0.2266 1 1>}
  NamedRerouteDeclaration() {name=RoughnessMask, nodeColor=<rgba 0.6641 0 1 1>}
  NamedRerouteUsage()
 ]}
